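# MetaSound patch that emits a series of repeat triggers at a randomized interval, with a configurable repeat count and a Stop input.
# Start draws a random time between Time.Min and Time.Max, which becomes the Trigger Repeat period and starts it.
# Each repeat also feeds back through a deferred trigger variable into the random node's Reset, not its Next, so only Start requests a new interval.
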MetaSoundPatch /FortMetasound/Patches/MSP_TriggerRepeatRandom
inputs [
 Num Repeats {type=Int32}
 Start {type=Trigger}
 Stop {type=Trigger}
 Time.Max {type=Time}
 Time.Min {type=Time}
]
outputs [Repeat Out {type=Trigger}]
pages [Default {
  input_defaults {Num Repeats=0, Start=null, Stop=null, Time.Max=1, Time.Min=0}
  nodes [
   n1 = VariableDeferredAccessor.Trigger(Variable=InitVariable.Trigger().Variable)
   n2 = UE.RandomTime(Next=^Start, Max=^Time.Max, Min=^Time.Min, Reset=@n1.Value)
   TriggerRepeat_0 = UE.TriggerRepeat(Num Repeats=^Num Repeats, Stop=^Stop, Start=@n2.On Next, Period=@n2.Value)
   VariableMutator.Trigger(Value=@TriggerRepeat_0.RepeatOut, Variable=@n1.Variable)
  ]
  links [{dst="@Repeat Out", src=TriggerRepeat_0.RepeatOut}]
 }]
