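# Looping mono MetaSound source for the ambient energy sound of weapon cases, with attenuation and a concurrency cap.
# On play it starts drone, whine, rattle and sizzle loops, randomizing the drone and whine start times.
# A randomized repeat trigger adds pitch-shifted swells drawn from a wave array with no-repeat selection.
# NumberOfCases scales the drone, whine, rattle and swell gains through map ranges, while sizzle is mixed in at fixed gain.
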MetaSoundSource /PaprikaWeaponCases/Audio/MSS_WeaponCase_Ambient_Loop
inputs [
 AdditionalOperands {type=Float}
 Cutoff Frequency {type=Float}
 Gain 0 {type=Float}
 Gain 1 {type=Float}
 In {type=Float}
 In 1 {type=Float}
 In 2 {type=Float}
 In 3 {type=Float}
 Int32 {type=Int32}
 Int32 1 {type=Int32}
 NumberOfCases {type=Int32}
 Out Range B {type=Float}
 Pitch Shift {type=Float}
 test {type=Float}
 UE.Source.OnPlay {type=Trigger}
 Wave Asset {type=WaveAsset}
 WaveArray {type=WaveAsset:Array}
]
interfaces [UE.OutputFormat.Mono:1.0, UE.Source:1.1]
outputs ["UE.OutputFormat.Mono.Audio:0" {type=Audio}]
pages [Default {
  input_defaults {
   AdditionalOperands=0
   Cutoff Frequency=200.3104
   Gain 0=0
   Gain 1=5
   In=1
   In 1=1
   In 2=1
   In 3=0
   Int32=0
   Int32 1=0
   NumberOfCases=0
   Out Range B=0.4
   Pitch Shift=0
   UE.Source.OnPlay=null
   Wave Asset=null
   test=8
   WaveArray [
    /PaprikaWeaponCases/Audio/Waves/WeaponCase_AmbientLoop_Energy_Swell_01
    /PaprikaWeaponCases/Audio/Waves/WeaponCase_AmbientLoop_Energy_Swell_02
    /PaprikaWeaponCases/Audio/Waves/WeaponCase_AmbientLoop_Energy_Swell_03
    /PaprikaWeaponCases/Audio/Waves/WeaponCase_AmbientLoop_Energy_Swell_04
    /PaprikaWeaponCases/Audio/Waves/WeaponCase_AmbientLoop_Energy_Swell_05
    /PaprikaWeaponCases/Audio/Waves/WeaponCase_AmbientLoop_Energy_Swell_06
   ]
  }
  nodes [
   n1 = InitVariable.Int32()
   n2 = Convert.Float:Int32(Int32=VariableAccessor.Int32(Variable=VariableMutator.Int32(Value=^NumberOfCases, Variable=@n1.Variable).Variable).Value)
   n3 = UE.RandomTime(Next=^UE.Source.OnPlay) {also {defaults {Max=20}}}
   WavePlayer_0 = UE.Wave Player:Mono(Play=@n3.On Next, Start Time=@n3.Value) {also {defaults {Loop="<MetaSoundBuilderResult.SUCCEEDED: 0>", Wave Asset=/PaprikaWeaponCases/Audio/Waves/WeaponCase_AmbientLoop_Energy_Drone_B}}}
   ArrayRandomGet_0 = Array.Random Get:WaveAsset:Array(
    In Array=^WaveArray
    Next=/FortMetasound/Patches/MSP_TriggerRepeatRandom(Start=^UE.Source.OnPlay) {also {defaults {Time.Max=16, Time.Min=8}}}.Repeat Out
   ) {also {defaults {No Repeats=4}}}
   n4 = UE.RandomFloat(Next=@ArrayRandomGet_0.On Next) {also {defaults {Max=1.5, Min=0.95}}}
   WavePlayer_1 = UE.Wave Player:Mono(Wave Asset=@ArrayRandomGet_0.Value, Play=@n4.On Next, Pitch Shift=@n4.Value)
   WavePlayer_2 = UE.Wave Player:Mono(Play=@n3.On Next) {also {defaults {Loop="<MetaSoundBuilderResult.SUCCEEDED: 0>", Wave Asset=/PaprikaWeaponCases/Audio/Waves/WeaponCase_AmbientLoop_Energy_Rattle}}}
   WavePlayer_4 = UE.Wave Player:Mono(Play=@n3.On Next, Start Time=@n3.Value) {also {defaults {Loop="<MetaSoundBuilderResult.SUCCEEDED: 0>", Wave Asset=/PaprikaWeaponCases/Audio/Waves/WeaponCase_AmbientLoop_Energy_Whine}}}
   AudioMixerAudioMixerMono5_0 = "AudioMixer.Audio Mixer (Mono, 5)"(
    Gain 3=MapRange.MapRange:Float(In=@n2.Float) {also {defaults {In Range A=3, In Range B=8, Out Range A=0.25, Out Range B=0.4}}}.Out Value
    Gain 2=MapRange.MapRange:Float(In=@n2.Float) {also {defaults {In Range B=5, Out Range A=0.3, Out Range B=0.5}}}.Out Value
    Gain 4=MapRange.MapRange:Float(In=@n2.Float) {also {defaults {In Range A=1, In Range B=8, Out Range A=0.1, Out Range B=0.15}}}.Out Value
    Gain 1=MapRange.MapRange:Float(In=@n2.Float) {also {defaults {In Range B=8, Out Range A=0.6, Out Range B=1.3}}}.Out Value
    In 4=@WavePlayer_0.Out Mono
    In 1=@WavePlayer_1.Out Mono
    In 3=@WavePlayer_2.Out Mono
    In 2=@WavePlayer_4.Out Mono
   ) {also {defaults {Gain 0=0.15}}}
   WavePlayer_3 = UE.Wave Player:Mono(Play=@n3.On Next) {also {defaults {Loop="<MetaSoundBuilderResult.SUCCEEDED: 0>", Wave Asset=/PaprikaWeaponCases/Audio/Waves/WeaponCase_AmbientLoop_Energy_Sizzle}}}
   AudioMixerAudioMixerMono2_0 = "AudioMixer.Audio Mixer (Mono, 2)"(In 0=@AudioMixerAudioMixerMono5_0.Out, In 1=@WavePlayer_3.Out Mono) {also {defaults {Gain 1=0.25}}}
   InitVariable.Float()
  ]
  links [{dst="@UE.OutputFormat.Mono.Audio:0", src=AudioMixerAudioMixerMono2_0.Out}]
 }]
properties {
 assetImportData=&self:AssetImportData
 attenuationSettings=&/Fort_Loot_Ambient_Loop_Mono_WeaponCase_Att
 bLooping=true
 bOverrideConcurrency=true
 duration=10000
 maxDistance=1500
 sampleRate=48000
 soundClassObject=/Game/Sounds/FortSoundClassesAndModes/Fort_Loot_Ambient_Loop
 concurrencyOverrides {maxCount=6}
}
