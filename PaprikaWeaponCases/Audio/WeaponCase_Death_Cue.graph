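# Sound cue for a weapon case death, playing the WeaponCase_Emptied_01 wave as a loot-reward sound.
# A single wave player feeds a modulator that randomizes volume and lowers pitch slightly.
# It uses an item-interact mono attenuation and the loot reward sound class, with a raised volume multiplier.
# Its concurrency override prevents new instances from starting once the limit is reached.
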SoundCue /PaprikaWeaponCases/Audio/WeaponCase_Death_Cue
root=Modulator_0
nodes [
 {id=Modulator_0, children [WavePlayer_0], properties {pitchMax=0.95, pitchMin=0.85, volumeMin=0.5}}
 {id=WavePlayer_0, properties {soundWaveAssetPtr=/PaprikaWeaponCases/Audio/Waves/WeaponCase_Emptied_01}}
]
properties {
 attenuationSettings=/Game/Sounds/Attenuation/Fort_Item_Interact_Mono_Att
 bOverrideConcurrency=true
 maxDistance=1500
 soundClassObject=&Fort_Loot_Reward_PSM
 volumeMultiplier=1.5
 concurrencyOverrides {resolutionRule=PreventNew}
}
