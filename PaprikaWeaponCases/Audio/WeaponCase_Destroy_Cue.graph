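# Sound cue that plays the destruction sound for a weapon case, classed as a loot reward sound.
# A Random node picks one of three equally weighted destroy waves and feeds a Modulator that varies playback, with a lowered minimum volume.
# It uses mono item-interact attenuation and a boosted volume.
# A concurrency override prevents new instances at the limit and sets a short retrigger time.
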SoundCue /PaprikaWeaponCases/Audio/WeaponCase_Destroy_Cue
root=Modulator_0
nodes [
 {id=Modulator_0, children [Random_0], properties {volumeMin=0.75}}
 {id=Random_0, children [WavePlayer_0,WavePlayer_1,WavePlayer_2], properties {weights [1,1,1]}}
 {id=WavePlayer_0, properties {soundWaveAssetPtr=/PaprikaWeaponCases/Audio/Waves/WeaponCase_Destroy_01}}
 {id=WavePlayer_1, properties {soundWaveAssetPtr=/PaprikaWeaponCases/Audio/Waves/WeaponCase_Destroy_02}}
 {id=WavePlayer_2, properties {soundWaveAssetPtr=/PaprikaWeaponCases/Audio/Waves/WeaponCase_Destroy_03}}
]
properties {
 attenuationSettings=/Game/Sounds/Attenuation/Fort_Item_Interact_Mono_Att
 bOverrideConcurrency=true
 maxDistance=1500
 soundClassObject=&Fort_Loot_Reward_PSM
 volumeMultiplier=1.5
 concurrencyOverrides {resolutionRule=PreventNew, retriggerTime=0.05}
}
