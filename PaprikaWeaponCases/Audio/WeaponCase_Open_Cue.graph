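# Sound cue for a weapon case opening, layering a base sound with a delayed "sweet" accent.
# A mixer blends both layers, each picking randomly from four waves through a modulator, with the accent slightly louder and starting after a short fixed delay.
# It raises overall pitch and volume, uses item-interact attenuation and the loot reward sound class, and overrides concurrency to block new instances at the limit.
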SoundCue /PaprikaWeaponCases/Audio/WeaponCase_Open_Cue
root=Mixer_0
nodes [
 {id=Delay_1, children [Modulator_1], properties {delayMax=0.2, delayMin=0.2}}
 {id=Mixer_0, children [Delay_1,Modulator_3], properties {inputVolume [1.2,1]}}
 {id=Modulator_1, children [Random_6], properties {pitchMax=1.2, pitchMin=1}}
 {id=Modulator_3, children [Random_0], properties {pitchMin=1}}
 {id=Random_0, children [WavePlayer_0,WavePlayer_2,WavePlayer_3,WavePlayer_4], properties {weights [1,1,1,1]}}
 {id=Random_6, children [WavePlayer_16, WavePlayer_17, WavePlayer_18, WavePlayer_19], properties {weights [1,1,1,1]}}
 {id=WavePlayer_0, properties {soundWaveAssetPtr=/PaprikaWeaponCases/Audio/Waves/WeaponCase_Open_Base_01}}
 {id=WavePlayer_16, properties {soundWaveAssetPtr=/PaprikaWeaponCases/Audio/Waves/WeaponCase_Open_Sweet_01}}
 {id=WavePlayer_17, properties {soundWaveAssetPtr=/PaprikaWeaponCases/Audio/Waves/WeaponCase_Open_Sweet_02}}
 {id=WavePlayer_18, properties {soundWaveAssetPtr=/PaprikaWeaponCases/Audio/Waves/WeaponCase_Open_Sweet_03}}
 {id=WavePlayer_19, properties {soundWaveAssetPtr=/PaprikaWeaponCases/Audio/Waves/WeaponCase_Open_Sweet_04}}
 {id=WavePlayer_2, properties {soundWaveAssetPtr=/PaprikaWeaponCases/Audio/Waves/WeaponCase_Open_Base_02}}
 {id=WavePlayer_3, properties {soundWaveAssetPtr=/PaprikaWeaponCases/Audio/Waves/WeaponCase_Open_Base_03}}
 {id=WavePlayer_4, properties {soundWaveAssetPtr=/PaprikaWeaponCases/Audio/Waves/WeaponCase_Open_Base_04}}
]
properties {
 attenuationSettings=/Game/Sounds/Attenuation/Fort_Item_Interact_Att
 bOverrideConcurrency=true
 maxDistance=1850
 pitchMultiplier=1.3
 soundClassObject=&Fort_Loot_Reward_PSM
 volumeMultiplier=2.4
 concurrencyOverrides {resolutionRule=PreventNew}
}
